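# Physics asset for the GumOutlaw pickaxe axe mesh, a two-handed melee cosmetic weapon.
# It holds a single capsule body on the Melee_TwoHanded_Handle bone, offset and rotated to wrap the weapon as a rough collision volume.
# It has no constraints, so the weapon acts as one rigid piece with no jointed or secondary motion.
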PhysicsAsset /BRCosmetics/Weapons/FORT_Melee/Pickaxe_GumOutlaw/Meshes/GumOutlaw_Axe_PhysicsAsset
bodies {Melee_TwoHanded_Handle [{shapeType=Capsule, length=65.0198, radius=30.4017, center=<0.7825 1.4982 29.9742>, rotation=<rot 3.2572 -90 0.0087>}]}
constraints []
